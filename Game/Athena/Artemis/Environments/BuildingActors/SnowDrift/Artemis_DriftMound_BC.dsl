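; Snow drift mound building actor that appears to breathe and snore, with steam puffs.
; BeginPlay creates a dynamic material from a mound-type map lookup, then, on authority, starts the breath loop after a random delay.
; The server toggles IsActive using the breath timeline's length and random pauses.
; Its OnRep plays that timeline only on remotes, animating the BreathingControl parameter and triggering snore audio and steam, keeping cosmetics client-side.
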(event EventBeginPlay
  (bind _moundtype (Variables|Default|GetMoundType))
  (bind _moundtypemap (Variables|Default|GetMoundTypeMap))
  (bind (_value _returnvalue) (Utilities|Map|Find _moundtypemap _moundtype))
  (if _returnvalue
    (bind _returnvalue_2 (Rendering|Material|CreateDynamicMaterialInstance (Variables|Default|GetBC_Mound) 0 _value "None" :!class /Script/Engine.PrimitiveComponent))
    (Variables|Default|SetMound_MID _returnvalue_2)
    (Networking|SwitchHasAuthority
      (:Authority
        (Utilities|FlowControl|Delay (Math|Random|RandomFloatinRange 1.0 3.0))
        (CallFunction|ServerBreath)))))

(event Custom|Server_Breath
  (Variables|Default|SetIsActive true)
  (Utilities|FlowControl|Delay (Components|Timeline|GetTimelineLength (Variables|Artemis_DriftMound_BC|GetBreath_Timeline)))
  (Variables|Default|SetIsActive false)
  (Utilities|FlowControl|Delay (Math|Random|RandomFloatinRange 0.2 1.0))
  (CallFunction|ServerBreath))

(event Custom|EnableBreath
  ;[0 1 0.1982 1 22] Breath
  (bind (_direction _breath_anim) (Timeline|Breath_Timeline PlayFromStart)
    (:Update
      (Rendering|Material|SetScalarParameterValue (Variables|Default|GetMound_MID) "BreathingControl" (Math|Float|Lerp 0.0 1.0 _breath_anim) :!class /Script/Engine.MaterialInstanceDynamic))
    (:Snore_Audio
      (Audio|Components|Audio|Play (Variables|Default|GetSnore) 0.0))
    (:ToggleSmoke
      (Components|Activation|Activate (Variables|Default|GetSteam) true))))

(fn OnRep_IsActive
  (Networking|SwitchHasAuthority
    (:Remote
      (if (Variables|Default|GetIsActive)
        ;[1 1 1] SFX/VFX only in client
        (CallFunction|EnableBreath)))))

(fn ConstructionScript)
